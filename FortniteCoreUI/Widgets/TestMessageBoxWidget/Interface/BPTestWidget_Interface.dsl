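; BPTestWidget_Interface is a Blueprint interface for a test message box widget in the FortniteCoreUI widgets.
; It declares three functions: addMsg takes a Test Message Struct, and showWidget and hideWidget take no parameters.
; It holds only signatures, so implementing widgets supply the message display and visibility logic.
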(fn addMsg
  (params (msg :type Test%20Message%20Struct%20Structure)))

(fn hideWidget)

(fn showWidget)
